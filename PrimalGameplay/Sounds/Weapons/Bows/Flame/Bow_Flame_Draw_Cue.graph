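# Sound cue for drawing the flame bow, classed as weapon foley with bow concurrency rules and weapon reload attenuation.
# A mixer layers one of three randomly chosen draw one-shots with a looping tension sound.
# An enveloper fades the loop's volume in over about half a second, and a modulator varies its raised pitch and its volume.
# Separate volume variation applies to the one-shot, and the cue's debug flag is left on.
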SoundCue /PrimalGameplay/Sounds/Weapons/Bows/Flame/Bow_Flame_Draw_Cue
root=Modulator_0
nodes [
 {
  id=Enveloper_0
  children [WavePlayer_0]
  properties {pitchCurve {editorCurveData {keys [{time=0}]}}, volumeCurve {editorCurveData {keys [<at 0 0>,<at 0.5291 0.9862>]}}}
 }
 {id=Mixer_0, children [Modulator_3,Modulator_2], properties {inputVolume [0.8,1]}}
 {id=Modulator_0, children [Mixer_0]}
 {id=Modulator_2, children [Enveloper_0], properties {pitchMax=1.2, pitchMin=1.1, volumeMax=1, volumeMin=0.9}}
 {id=Modulator_3, children [Random_2], properties {volumeMax=0.85, volumeMin=0.75}}
 {id=Random_2, children [WavePlayer_10, WavePlayer_11, WavePlayer_12], properties {weights [1,1,1]}}
 {id=WavePlayer_0, properties {bLooping=true, soundWaveAssetPtr=&/Bow_Flame_Draw_TensionLoop_02}}
 {id=WavePlayer_10, properties {soundWaveAssetPtr=&/Bow_Flame_Draw_02}}
 {id=WavePlayer_11, properties {soundWaveAssetPtr=&/Bow_Flame_Draw_01}}
 {id=WavePlayer_12, properties {soundWaveAssetPtr=&/Bow_Flame_Draw_03}}
]
properties {
 attenuationSettings=&Fort_Weapon_Reload_Att
 bDebug=true
 maxDistance=2250
 soundClassObject=&Fort_WeaponFoley
 volumeMultiplier=0.301
 concurrencySet [/Game/Sounds/Concurrency_Rules/Bow_Global_Conc, /Game/Sounds/Concurrency_Rules/Bow_Local_Conc]
}
