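# Sound cue for dropping a Flame Bow weapon, using the weapon-drop attenuation and the loot-collect sound class.
# A Random node picks one of three drop waves with equal weight, feeding a Modulator that varies volume and slightly lowers pitch.
# Cue-level multipliers further reduce pitch and volume, and global and local bow concurrency rules limit overlapping plays.
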SoundCue /PrimalGameplay/Sounds/Weapons/Bows/Flame/Bow_Flame_Drop_Cue
root=Modulator_0
nodes [
 {id=Modulator_0, children [Random_1], properties {pitchMax=0.95, pitchMin=0.85, volumeMax=1.1}}
 {id=Random_1, children [WavePlayer_1,WavePlayer_4,WavePlayer_5], properties {weights [1,1,1]}}
 {id=WavePlayer_1, properties {soundWaveAssetPtr=&/Bow_Flame_Drop_01}}
 {id=WavePlayer_4, properties {soundWaveAssetPtr=&/Bow_Flame_Drop_02}}
 {id=WavePlayer_5, properties {soundWaveAssetPtr=&/Bow_Flame_Drop_03}}
]
properties {
 attenuationSettings=&Fort_Weapon_Drop_Att
 maxDistance=3000
 pitchMultiplier=0.9
 soundClassObject=&Fort_Loot_Collect
 volumeMultiplier=0.8
 concurrencySet [/Game/Sounds/Concurrency_Rules/Bow_Global_Conc, /Game/Sounds/Concurrency_Rules/Bow_Local_Conc]
}
