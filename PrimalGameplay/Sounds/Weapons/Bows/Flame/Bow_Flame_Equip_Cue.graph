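# Sound cue for equipping the flame bow, a short weapon foley one-shot.
# A Random node picks among three equip wave variations with equal weight, then a Modulator applies slight pitch and volume variation.
# It plays in the weapon foley sound class with reload-style attenuation, a reduced volume, and global and local bow concurrency rules.
# The cue's debug flag is left enabled.
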SoundCue /PrimalGameplay/Sounds/Weapons/Bows/Flame/Bow_Flame_Equip_Cue
root=Modulator_1
nodes [
 {id=Modulator_1, children [Random_1], properties {pitchMax=1.1, pitchMin=0.9, volumeMax=0.95, volumeMin=0.85}}
 {id=Random_1, children [WavePlayer_0,WavePlayer_4,WavePlayer_5], properties {weights [1,1,1]}}
 {id=WavePlayer_0, properties {soundWaveAssetPtr=&/Bow_Flame_Equip_01}}
 {id=WavePlayer_4, properties {soundWaveAssetPtr=&/Bow_Flame_Equip_02}}
 {id=WavePlayer_5, properties {soundWaveAssetPtr=&/Bow_Flame_Equip_03}}
]
properties {
 attenuationSettings=&Fort_Weapon_Reload_Att
 bDebug=true
 maxDistance=2250
 soundClassObject=&Fort_WeaponFoley
 volumeMultiplier=0.351
 concurrencySet [/Game/Sounds/Concurrency_Rules/Bow_Global_Conc, /Game/Sounds/Concurrency_Rules/Bow_Local_Conc]
}
