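# Sound cue for the flame bow's fire impact, routed to the Fort_Impacts sound class.
# A mixer layers two modulated random branches: one of four flame fire-impact waves, and one of three charged metal-bow projectile impacts reused at slightly lower level.
# The metal layer gets slight upward pitch variation, and the whole cue is boosted in volume and pitch.
# Arrow-impact attenuation and global and local bow concurrency rules apply.
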SoundCue /PrimalGameplay/Sounds/Weapons/Bows/Flame/Bow_Flame_FireImpact_Cue
root=Mixer_0
nodes [
 {id=Mixer_0, children [Modulator_2,Modulator_0], properties {inputVolume [1,0.9]}}
 {id=Modulator_0, children [Random_2], properties {pitchMax=1.1, pitchMin=1}}
 {id=Modulator_2, children [Random_0]}
 {id=Random_0, children [WavePlayer_2,WavePlayer_3,WavePlayer_5,WavePlayer_6], properties {weights [1,1,1,1]}}
 {id=Random_2, children [WavePlayer_7, WavePlayer_14, WavePlayer_15], properties {weights [1,1,1]}}
 {id=WavePlayer_14, properties {soundWaveAssetPtr=/PrimalGameplay/Sounds/Weapons/Bows/Metal/Bow_Metal_Prj_Charged_Impact_03}}
 {id=WavePlayer_15, properties {soundWaveAssetPtr=/PrimalGameplay/Sounds/Weapons/Bows/Metal/Bow_Metal_Prj_Charged_Impact_04}}
 {id=WavePlayer_2, properties {soundWaveAssetPtr=&/Bow_Flame_FireImpact_01}}
 {id=WavePlayer_3, properties {soundWaveAssetPtr=&/Bow_Flame_FireImpact_02}}
 {id=WavePlayer_5, properties {soundWaveAssetPtr=&/Bow_Flame_FireImpact_03}}
 {id=WavePlayer_6, properties {soundWaveAssetPtr=&/Bow_Flame_FireImpact_04}}
 {id=WavePlayer_7, properties {soundWaveAssetPtr=/PrimalGameplay/Sounds/Weapons/Bows/Metal/Bow_Metal_Prj_Charged_Impact_01}}
]
properties {
 attenuationSettings=/Game/Sounds/Attenuation/Fort_Arrow_Impact_Athena_Att
 maxDistance=6000
 pitchMultiplier=1.1
 soundClassObject=&Fort_Impacts
 volumeMultiplier=2.4192
 attenuationOverrides {falloffDistance=2200, reverbWetLevelMin=0.2, attenuationShapeExtents {x=600}}
 concurrencySet [/Game/Sounds/Concurrency_Rules/Bow_Global_Conc, /Game/Sounds/Concurrency_Rules/Bow_Local_Conc]
}
