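# Sound cue for the flame bow's full-tension firing shot, blending a close layer and a distant layer.
# Each layer randomly picks one of four waves: the main set up close, with pitched-down, boosted modulation, and "Medium" variants at range.
# Each layer has its own projectile attenuation and sound class override before a mixer weights the close layer higher.
# A root modulator wraps the mix, and global and local bow concurrency rules govern playback.
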SoundCue /PrimalGameplay/Sounds/Weapons/Bows/Flame/Bow_Flame_Fire_Full_Tension_Cue
root=Modulator_3
nodes [
 {
  id=Attenuation_3
  children [Modulator_4]
  properties {attenuationSettings=/Game/Sounds/Attenuation/Fort_Weapon_Projectile_Close_Bow_Athena_Att}
 }
 {
  id=Attenuation_4
  children [Modulator_6]
  properties {attenuationSettings=/Game/Sounds/Attenuation/Fort_Weapon_Projectile_Distant_Bow_Athena_Att}
 }
 {id=Mixer_1, children [SoundClass_3,SoundClass_2], properties {inputVolume [0.7,0.4]}}
 {id=Modulator_3, children [Mixer_1]}
 {id=Modulator_4, children [Random_0], properties {pitchMax=0.95, pitchMin=0.85, volumeMax=1.3, volumeMin=1.2}}
 {id=Modulator_6, children [Random_4]}
 {id=Random_0, children [WavePlayer_8, WavePlayer_10, WavePlayer_11, WavePlayer_12], properties {weights [1,1,1,1]}}
 {id=Random_4, children [WavePlayer_17, WavePlayer_18, WavePlayer_19, WavePlayer_20], properties {weights [1,1,1,1]}}
 {
  id=SoundClass_2
  children [Attenuation_4]
  properties {soundClassOverride=&Fort_WeaponRanged_Fire_Distant}
 }
 {
  id=SoundClass_3
  children [Attenuation_3]
  properties {soundClassOverride=&Fort_WeaponRanged_Fire_3P_PSM}
 }
 {id=WavePlayer_10, properties {soundWaveAssetPtr=&/Bow_Flame_Fire_Full_Tension_02}}
 {id=WavePlayer_11, properties {soundWaveAssetPtr=&/Bow_Flame_Fire_Full_Tension_03}}
 {id=WavePlayer_12, properties {soundWaveAssetPtr=&/Bow_Flame_Fire_Full_Tension_04}}
 {id=WavePlayer_17, properties {soundWaveAssetPtr=&/Bow_Flame_Fire_Full_Tension_Medium_01}}
 {id=WavePlayer_18, properties {soundWaveAssetPtr=&/Bow_Flame_Fire_Full_Tension_Medium_02}}
 {id=WavePlayer_19, properties {soundWaveAssetPtr=&/Bow_Flame_Fire_Full_Tension_Medium_03}}
 {id=WavePlayer_20, properties {soundWaveAssetPtr=&/Bow_Flame_Fire_Full_Tension_Medium_04}}
 {id=WavePlayer_8, properties {soundWaveAssetPtr=&/Bow_Flame_Fire_Full_Tension_01}}
]
properties {
 maxDistance=27500
 volumeMultiplier=1.1
 attenuationOverrides {falloffDistance=5000, attenuationShapeExtents {x=1000}}
 concurrencySet [/Game/Sounds/Concurrency_Rules/Bow_Global_Conc, /Game/Sounds/Concurrency_Rules/Bow_Local_Conc]
}
