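# First-person sound cue for firing the flame bow at low tension, played at half volume.
# A random node picks one of four equally weighted wave variants, which feeds through two chained modulators, the outer one widening pitch variation.
# Playback uses a first-person bow projectile attenuation asset, a ranged-weapon fire sound class and first-person gunfire concurrency.
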SoundCue /PrimalGameplay/Sounds/Weapons/Bows/Flame/Bow_Flame_Fire_Low_Tension_1P_Cue
root=Modulator_0
nodes [
 {id=Modulator_0, children [Modulator_1], properties {pitchMax=1.1, pitchMin=0.9}}
 {id=Modulator_1, children [Random_1]}
 {id=Random_1, children [WavePlayer_4,WavePlayer_5,WavePlayer_6,WavePlayer_7], properties {weights [1,1,1,1]}}
 {id=WavePlayer_4, properties {soundWaveAssetPtr=&/Bow_Flame_Fire_Low_Tension_01}}
 {id=WavePlayer_5, properties {soundWaveAssetPtr=&/Bow_Flame_Fire_Low_Tension_02}}
 {id=WavePlayer_6, properties {soundWaveAssetPtr=&/Bow_Flame_Fire_Low_Tension_03}}
 {id=WavePlayer_7, properties {soundWaveAssetPtr=&/Bow_Flame_Fire_Low_Tension_04}}
]
properties {
 attenuationSettings=/Game/Sounds/Attenuation/Fort_Weapon_Projectile_FirstPerson_Bow_Athena_Att
 maxDistance=2097152
 soundClassObject=&Fort_WeaponRanged_Fire_1P_PSM
 volumeMultiplier=0.5
 attenuationOverrides {falloffDistance=5000, attenuationShapeExtents {x=1000}}
 concurrencySet [&Weapon_Gunfire_FirstPerson_Conc]
}
