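# Sound cue for firing the flame bow at low tension, split into close and distant layers.
# The close branch randomly picks one of four low-tension fire waves, using close-bow attenuation and a third-person fire sound class.
# The quieter distant branch mixes four random medium variants with a pitch-modulated explosive-bow shot, under distant attenuation and its own sound class.
# A root modulator wraps both layers, and global and local bow concurrency rules apply.
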SoundCue /PrimalGameplay/Sounds/Weapons/Bows/Flame/Bow_Flame_Fire_Low_Tension_Cue
root=Modulator_3
nodes [
 {
  id=Attenuation_3
  children [Modulator_4]
  properties {attenuationSettings=/Game/Sounds/Attenuation/Fort_Weapon_Projectile_Close_Bow_Athena_Att}
 }
 {
  id=Attenuation_4
  children [Mixer_2]
  properties {attenuationSettings=/Game/Sounds/Attenuation/Fort_Weapon_Projectile_Distant_Bow_Athena_Att}
 }
 {id=Mixer_1, children [SoundClass_3,SoundClass_2], properties {inputVolume [0.7,0.4]}}
 {id=Mixer_2, children [Modulator_5,Modulator_6], properties {inputVolume [1,0.6]}}
 {id=Modulator_3, children [Mixer_1]}
 {id=Modulator_4, children [Random_1], properties {volumeMax=1.2, volumeMin=1.1}}
 {id=Modulator_5, children [Random_4]}
 {id=Modulator_6, children [WavePlayer_23], properties {pitchMax=1, pitchMin=0.88, volumeMax=0.8, volumeMin=0.7}}
 {id=Random_1, children [WavePlayer_2, WavePlayer_10, WavePlayer_11, WavePlayer_12], properties {weights [1,1,1,1]}}
 {id=Random_4, children [WavePlayer_19, WavePlayer_20, WavePlayer_21, WavePlayer_22], properties {weights [1,1,1,1]}}
 {
  id=SoundClass_2
  children [Attenuation_4]
  properties {soundClassOverride=&Fort_WeaponRanged_Fire_Distant}
 }
 {
  id=SoundClass_3
  children [Attenuation_3]
  properties {soundClassOverride=&Fort_WeaponRanged_Fire_3P_PSM}
 }
 {id=WavePlayer_10, properties {soundWaveAssetPtr=&/Bow_Flame_Fire_Low_Tension_02}}
 {id=WavePlayer_11, properties {soundWaveAssetPtr=&/Bow_Flame_Fire_Low_Tension_03}}
 {id=WavePlayer_12, properties {soundWaveAssetPtr=&/Bow_Flame_Fire_Low_Tension_04}}
 {id=WavePlayer_19, properties {soundWaveAssetPtr=&/Bow_Flame_Fire_Low_Tension_Medium_01}}
 {id=WavePlayer_2, properties {soundWaveAssetPtr=&/Bow_Flame_Fire_Low_Tension_01}}
 {id=WavePlayer_20, properties {soundWaveAssetPtr=&/Bow_Flame_Fire_Low_Tension_Medium_02}}
 {id=WavePlayer_21, properties {soundWaveAssetPtr=&/Bow_Flame_Fire_Low_Tension_Medium_03}}
 {id=WavePlayer_22, properties {soundWaveAssetPtr=&/Bow_Flame_Fire_Low_Tension_Medium_04}}
 {id=WavePlayer_23, properties {soundWaveAssetPtr=/Game/Sounds/Fort_Weapons/Explosive_Bow/Weapon_Bow_Compound_Shoot_Med_01}}
]
properties {
 maxDistance=27500
 volumeMultiplier=1
 attenuationOverrides {falloffDistance=5000, attenuationShapeExtents {x=1000}}
 concurrencySet [/Game/Sounds/Concurrency_Rules/Bow_Global_Conc, /Game/Sounds/Concurrency_Rules/Bow_Local_Conc]
}
