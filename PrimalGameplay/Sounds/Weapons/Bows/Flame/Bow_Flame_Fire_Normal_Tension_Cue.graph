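# Sound cue for firing the Flame Bow at normal tension, layering close and distant versions of the shot.
# The close layer randomizes four tension shots through close projectile attenuation and the third-person fire sound class.
# The distant layer mixes four random medium variants with a pitch-varied explosive-bow compound shot, then applies distant attenuation, the distant fire class and lowered pitch.
# Global and local bow concurrency sets limit playback.
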SoundCue /PrimalGameplay/Sounds/Weapons/Bows/Flame/Bow_Flame_Fire_Normal_Tension_Cue
root=Modulator_3
nodes [
 {
  id=Attenuation_3
  children [Modulator_4]
  properties {attenuationSettings=/Game/Sounds/Attenuation/Fort_Weapon_Projectile_Close_Bow_Athena_Att}
 }
 {
  id=Attenuation_4
  children [Mixer_2]
  properties {attenuationSettings=/Game/Sounds/Attenuation/Fort_Weapon_Projectile_Distant_Bow_Athena_Att}
 }
 {id=Mixer_1, children [SoundClass_3,Modulator_0], properties {inputVolume [0.7,0.4]}}
 {id=Mixer_2, children [Random_4,Modulator_5], properties {inputVolume [1,0.8]}}
 {id=Modulator_0, children [SoundClass_2], properties {pitchMax=0.9, pitchMin=0.8}}
 {id=Modulator_3, children [Mixer_1]}
 {id=Modulator_4, children [Random_0], properties {volumeMax=0.85, volumeMin=0.75}}
 {id=Modulator_5, children [WavePlayer_13], properties {pitchMax=1, pitchMin=0.8, volumeMax=0.9, volumeMin=0.7}}
 {id=Random_0, children [WavePlayer_5, WavePlayer_6, WavePlayer_8, WavePlayer_12], properties {weights [1,1,1,1]}}
 {id=Random_4, children [WavePlayer_22, WavePlayer_23, WavePlayer_24, WavePlayer_25], properties {weights [1,1,1,1]}}
 {
  id=SoundClass_2
  children [Attenuation_4]
  properties {soundClassOverride=&Fort_WeaponRanged_Fire_Distant}
 }
 {
  id=SoundClass_3
  children [Attenuation_3]
  properties {soundClassOverride=&Fort_WeaponRanged_Fire_3P_PSM}
 }
 {id=WavePlayer_12, properties {soundWaveAssetPtr=&/Bow_Flame_Fire_Normal_Tension_04}}
 {id=WavePlayer_13, properties {soundWaveAssetPtr=/Game/Sounds/Fort_Weapons/Explosive_Bow/Weapon_Bow_Compound_Shoot_Med_01}}
 {id=WavePlayer_22, properties {soundWaveAssetPtr=&/Bow_Flame_Fire_Normal_Tension_Medium_02}}
 {id=WavePlayer_23, properties {soundWaveAssetPtr=&/Bow_Flame_Fire_Normal_Tension_Medium_01}}
 {id=WavePlayer_24, properties {soundWaveAssetPtr=&/Bow_Flame_Fire_Normal_Tension_Medium_03}}
 {id=WavePlayer_25, properties {soundWaveAssetPtr=&/Bow_Flame_Fire_Normal_Tension_Medium_04}}
 {id=WavePlayer_5, properties {soundWaveAssetPtr=&/Bow_Flame_Fire_Normal_Tension_01}}
 {id=WavePlayer_6, properties {soundWaveAssetPtr=&/Bow_Flame_Fire_Normal_Tension_02}}
 {id=WavePlayer_8, properties {soundWaveAssetPtr=&/Bow_Flame_Fire_Normal_Tension_03}}
]
properties {
 maxDistance=27500
 volumeMultiplier=1
 attenuationOverrides {falloffDistance=5000, attenuationShapeExtents {x=1000}}
 concurrencySet [/Game/Sounds/Concurrency_Rules/Bow_Global_Conc, /Game/Sounds/Concurrency_Rules/Bow_Local_Conc]
}
