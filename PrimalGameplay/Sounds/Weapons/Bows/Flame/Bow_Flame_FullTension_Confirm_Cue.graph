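# Sound cue for the flame bow's confirmation sound at full draw tension.
# A mixer layers both default bow full-tension confirm waves together, then blends them with one of three randomly picked flame tension-start waves, weighted slightly louder.
# A root modulator randomizes pitch upward on each play.
# It uses weapon-reload attenuation, the weapon foley sound class and global and local bow concurrency rules, with the debug flag left on.
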SoundCue /PrimalGameplay/Sounds/Weapons/Bows/Flame/Bow_Flame_FullTension_Confirm_Cue
root=Modulator_0
nodes [
 {id=Mixer_0, children [Mixer_1,Random_2], properties {inputVolume [0.7,1.1]}}
 {id=Mixer_1, children [WavePlayer_1,WavePlayer_2], properties {inputVolume [1,1]}}
 {id=Modulator_0, children [Mixer_0], properties {pitchMax=1.2, pitchMin=1}}
 {id=Random_2, children [WavePlayer_7,WavePlayer_8,WavePlayer_9], properties {weights [1,1,1]}}
 {id=WavePlayer_1, properties {soundWaveAssetPtr=/PrimalGameplay/Sounds/Weapons/Bows/Default/Bow_Default_FullTension_Confirm_01}}
 {id=WavePlayer_2, properties {soundWaveAssetPtr=/PrimalGameplay/Sounds/Weapons/Bows/Default/Bow_Default_FullTension_Confirm_02}}
 {id=WavePlayer_7, properties {soundWaveAssetPtr=&/Bow_Flame_Draw_TensionStart_02}}
 {id=WavePlayer_8, properties {soundWaveAssetPtr=&/Bow_Flame_Draw_TensionStart_03}}
 {id=WavePlayer_9, properties {soundWaveAssetPtr=&/Bow_Flame_Draw_TensionStart_04}}
]
properties {
 attenuationSettings=&Fort_Weapon_Reload_Att
 bDebug=true
 maxDistance=2250
 soundClassObject=&Fort_WeaponFoley
 volumeMultiplier=0.451
 concurrencySet [/Game/Sounds/Concurrency_Rules/Bow_Global_Conc, /Game/Sounds/Concurrency_Rules/Bow_Local_Conc]
}
