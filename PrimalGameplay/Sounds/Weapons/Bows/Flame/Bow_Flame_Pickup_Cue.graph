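# Sound cue for the pickup sound of the flame bow weapon, classed as loot collection audio.
# A Random node picks between two pickup wave variants with equal weight and feeds a Modulator at the root, which is left at default settings.
# The cue plays at reduced volume through the shared weapon pickup attenuation, with global and local bow concurrency rules limiting simultaneous instances.
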SoundCue /PrimalGameplay/Sounds/Weapons/Bows/Flame/Bow_Flame_Pickup_Cue
root=Modulator_0
nodes [
 {id=Modulator_0, children [Random_0]}
 {id=Random_0, children [WavePlayer_0,WavePlayer_1], properties {weights [1,1]}}
 {id=WavePlayer_0, properties {soundWaveAssetPtr=&/Bow_Flame_Pickup_01}}
 {id=WavePlayer_1, properties {soundWaveAssetPtr=&/Bow_Flame_Pickup_02}}
]
properties {
 attenuationSettings=&Fort_Weapon_Pickup_Att
 maxDistance=3500
 soundClassObject=&Fort_Loot_Collect
 volumeMultiplier=0.5538
 concurrencySet [/Game/Sounds/Concurrency_Rules/Bow_Global_Conc, /Game/Sounds/Concurrency_Rules/Bow_Local_Conc]
}
