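# Sound cue for the flame bow's reload foley, playing one of three reload waves picked at random with equal weight.
# A modulator above the random node varies pitch slightly on each play.
# It uses the weapon reload attenuation and weapon foley sound class at reduced volume, with bow global and local concurrency rules limiting overlapping instances.
# The debug flag is left enabled on the cue.
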SoundCue /PrimalGameplay/Sounds/Weapons/Bows/Flame/Bow_Flame_Reload_Cue
root=Modulator_0
nodes [
 {id=Modulator_0, children [Random_1], properties {pitchMax=1.1, pitchMin=0.9}}
 {id=Random_1, children [WavePlayer_0,WavePlayer_1,WavePlayer_5], properties {weights [1,1,1]}}
 {id=WavePlayer_0, properties {soundWaveAssetPtr=&/Bow_Flame_Reload_01}}
 {id=WavePlayer_1, properties {soundWaveAssetPtr=&/Bow_Flame_Reload_02}}
 {id=WavePlayer_5, properties {soundWaveAssetPtr=&/Bow_Flame_Reload_03}}
]
properties {
 attenuationSettings=&Fort_Weapon_Reload_Att
 bDebug=true
 maxDistance=2250
 soundClassObject=&Fort_WeaponFoley
 volumeMultiplier=0.251
 concurrencySet [/Game/Sounds/Concurrency_Rules/Bow_Global_Conc, /Game/Sounds/Concurrency_Rules/Bow_Local_Conc]
}
